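; Malfunctioning vending machine station prop that adds a glitch effect when its service is used.
; Its PlayServiceUsedEffects override runs the parent version, then spawns the NS_VendingMachineGlitch effect at the static mesh's world location.
; The effect's rotation follows the mesh's forward vector, with yaw offset by a quarter turn.
; BeginPlay, overlap and Tick are disabled stubs, and the construction script only calls its parent.
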(event EventBeginPlay
  (|CallParentFunction) ; This node is disabled and will not be called.\nDrag off pins to build functionality.
)

(event Collision|EventActorBeginOverlap
  (params (OtherActor :type Actor%20Object%20Reference))
  (|CallParentFunction OtherActor) ; This node is disabled and will not be called.\nDrag off pins to build functionality.
)

(event EventTick
  (params (DeltaSeconds :type Float%20%28single-precision%29))
  (|CallParentFunction DeltaSeconds) ; This node is disabled and will not be called.\nDrag off pins to build functionality.
)

(fn PlayServiceUsedEffects
  (params (PlayerController :type Fort%20Player%20Controller%20Object%20Reference))
  (|CallParentFunction PlayerController)
  (Effects|Components|ParticleSystem|SpawnFXAtLocation "/Game/Effects/Fort_Effects/Gameplay/VendingMachine/NS_VendingMachineGlitch.NS_VendingMachineGlitch" (Transformation|GetWorldLocation (Variables|Debug|GetStaticMeshComponent)) (Math|Rotator|MakeRotator (.roll (Math|Conversions|RotationFromXVector (Transformation|GetForwardVector (Variables|Debug|GetStaticMeshComponent)))) (.pitch (Math|Conversions|RotationFromXVector (Transformation|GetForwardVector (Variables|Debug|GetStaticMeshComponent)))) (+ (.yaw (Math|Conversions|RotationFromXVector (Transformation|GetForwardVector (Variables|Debug|GetStaticMeshComponent)))) 9e+01))))

(fn ConstructionScript
  (|CallParentFunction))
